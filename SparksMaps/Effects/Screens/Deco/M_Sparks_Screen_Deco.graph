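# Unlit, animated deco screen material for the Sparks stage, pulsing to music with stage-manager or parameter-set colors.
# One deco texture supplies pattern, blur and gradient masks, combined with panning noise, a rotating-noise color cloud and a moving radial gradient.
# Audio-collection beat and measure lifetimes drive pulsing and gradient movement, with a fake-time switch for testing.
# The brightness-scaled sum passes through the shared screen-details function.
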Material /SparksMaps/Effects/Screens/Deco/M_Sparks_Screen_Deco
graph {
 properties {bAutomaticallySetUsageInEditor=false, bHasPixelAnimation=true, shadingModel=MSM_Unlit, usedShadingModels=Unlit}
 nodes [
  Constant4Vector() {constant=<rgba 0.8125 0.0073 0.1938 1>}
  Constant4Vector() {constant=<rgba 0.1116 0 1 1>}
  n1 = LinearInterpolate(A=NamedRerouteUsage(), Alpha=NamedRerouteUsage()) {constB=0}
  n2 = Constant3Vector() {constant=<rgba 1 0.2825 0.0997 1>}
  n3 = Multiply(A=NamedRerouteUsage(), B=@n2)
  n4 = ScalarParameter() {defaultValue=0.5, parameterName=Pattern Blur Beat Low}
  n5 = Constant() {r=0.75}
  n6 = Constant() {r=1}
  n7 = Add(
   A=Multiply(A=TextureCoordinate(), B=AppendVector(A=@n5, B=@n6)) {desc=Scale}
   B=AppendVector(A=ScalarParameter() {defaultValue=0.127, parameterName=Position X}, B=ScalarParameter() {parameterName=Position Y}) {desc=Position}
  )
  n8 = ScalarParameter() {defaultValue=1, parameterName=Scale By Center}
  n9 = AppendVector(A=@n8, B=@n8)
  n10 = TextureSample(UVs=NamedRerouteDeclaration(None=MaterialFunctionCall(UVs=@n7, Texture Scale=@n9) {materialFunction=&ScaleUVsByCenter}.UVs) {name=TexPosition, nodeColor=<rgba 0 1 0.8359 1>}) {texture=&/T_Sparks_Screen_Deco}
  n11 = Panner(Coordinate=TextureCoordinate() {uTiling=1.5, vTiling=1.5}) {speedY=0.1}
  n12 = Multiply(A=LinearInterpolate(
    A=Constant3Vector() {constant=<rgba 0.04 0.02 0 1>}
    Alpha=TextureSample(UVs=@n11) {samplerType=SAMPLERTYPE_Grayscale, texture=/Game/Global/Textures/Default/Noise/LowResBlurredNoiseNormalized}.R
   )) {constB=1.5, desc=Noise brightness}
  n13 = LinearInterpolate(A=NamedRerouteUsage(), B=Multiply(A=@n2, B=Multiply(A=@n10.R, B=@n12)), Alpha=NamedRerouteUsage())
  n14 = ScalarParameter() {defaultValue=0.4, parameterName=Pattern Beat Low}
  n15 = Add(
   A=LinearInterpolate(A=@n3, B=Multiply(A=@n4, B=@n3) {constA=0.1}, Alpha=NamedRerouteUsage())
   B=LinearInterpolate(A=@n13, B=Multiply(A=@n13, B=@n14) {constB=0.4}, Alpha=NamedRerouteUsage())
  )
  MaterialFunctionCall!(Emissive=Multiply(A=Add(A=@n1, B=@n15), B=ScalarParameter() {defaultValue=1, parameterName=OverallBrightness}), Scanlines Tile) {materialFunction=/SparksMaps/Effects/Screens/Materials/MF/MF_Sparks_Screen_Details}
  n16 = ScalarParameter() {defaultValue=1, parameterName=BeatGradientColor_Intensity}
  n17 = Multiply(A=NamedRerouteUsage(), B=@n16)
  n18 = Multiply(A=@n16, B=NamedRerouteUsage())
  n19 = TextureCoordinate()
  n20 = AppendVector(A=ScalarParameter() {defaultValue=1, parameterName=PatternGradient_X}, B=Constant() {r=1})
  n21 = ComponentMask(None=SmoothStep(
    Min=ScalarParameter() {defaultValue=0.2, parameterName=PatternColorGradient_Min}
    Max=ScalarParameter() {defaultValue=0.6, parameterName=PatternColorGradinet_Max}
    Value=Multiply(A=@n19, B=@n20)
   ) {constMax=0.6, constMin=0.5}) {r=true}
  n22 = Constant() {r=0.8}
  n23 = Constant() {r=1}
  n24 = Time() {desc=Fake Time for Testing}
  n25 = LinearInterpolate(
   A=CollectionParameter() {collection=/FMAudioReactiveData/AudioAnalysis/MPC_Sparks_Audio_Parameters, parameterName=MeasureLifetime}
   B=Frac(None=Multiply(A=@n24) {constB=0.5})
   Alpha=ScalarParameter() {parameterName=Real or Test Beat}
  ) {constAlpha=0}
  n26 = MaterialFunctionCall(
   UVs=Multiply(A=TextureCoordinate() {uTiling=2}, B=AppendVector(A=@n22, B=@n23))
   CenterPosition=AppendVector(A=Constant() {r=0.8}, B=LinearInterpolate(Alpha=@n25) {constA=0.5, constB=2})
   Radius=ScalarParameter() {defaultValue=0.8, parameterName=Radius}
   Density=ScalarParameter() {defaultValue=1.5, parameterName=Density}
  ) {materialFunction=&RadialGradientExponential}
  n27 = NamedRerouteUsage()
  n28 = LinearInterpolate(
   B=Multiply(A=LinearInterpolate(A=@n17, B=@n18, Alpha=@n21), B=@n16) {constB=0.5}
   Alpha=LinearInterpolate(
    B=Multiply(A=@n26.RadialGradientExponential)
    Alpha=TextureSample(UVs=@n27) {texture=&T_Sparks_Screen_Deco}.B
   )
  )
  n29 = LinearInterpolate(A=@n28, B=Multiply(A=@n28, B=ScalarParameter() {defaultValue=1, parameterName=Gradient Beat Low}) {constB=0.5}, Alpha=NamedRerouteUsage()) {constA=0.5, constAlpha=1, constB=0}
  n30 = Multiply(A=NamedRerouteUsage()) {constB=10}
  NamedRerouteDeclaration(None=Saturate(None=Multiply(A=@n29, B=@n30))) {name=BeatGradient, nodeColor=<rgba 1 0 0.8437 1>}
  n31 = TextureCoordinate()
  n32 = Multiply(A=Time()) {constB=0.5}
  n33 = Multiply(
   A=TextureSample(UVs=Rotator(Coordinate=@n31, Time=@n32)) {samplerType=SAMPLERTYPE_Grayscale, texture=&LowResBlurredNoiseNormalized}.R
   B=TextureSample(UVs=@n31) {texture=/Game/Effects/Fort_Effects/Emotes/Candor/T_Noise_Blur_Texture}.R
  )
  n34 = Multiply(
   A=@n33
   B=OneMinus(None=MaterialFunctionCall() {materialFunction=&RadialGradientExponential}.RadialGradientExponential)
  )
  n35 = Divide(
   A=Subtract(
    A=MaterialFunctionCall() {materialFunction=&LinearGradient}.VGradient
    B=ScalarParameter() {parameterName=NoiseGradSub}
   )
   B=ScalarParameter() {defaultValue=1, parameterName=NoiseGradDiv}
  )
  n36 = NamedRerouteUsage()
  n37 = NamedRerouteUsage()
  n38 = ComponentMask(None=TextureCoordinate()) {r=true}
  n39 = Multiply(
   A=Multiply(A=Multiply(A=@n34, B=@n35), B=LinearInterpolate(A=@n36, B=@n37, Alpha=@n38))
   B=ScalarParameter() {defaultValue=4, parameterName=ColorBoost}
  )
  n40 = Multiply(A=ScalarParameter() {defaultValue=0.8, parameterName=Cloud Color Beat Low}, B=@n39) {constA=0.4}
  n41 = NamedRerouteUsage()
  n42 = LinearInterpolate(A=NamedRerouteUsage(), B=NamedRerouteUsage(), Alpha=ComponentMask(None=TextureCoordinate()) {r=true})
  NamedRerouteDeclaration(None=Add(A=LinearInterpolate(A=@n39, B=@n40, Alpha=@n41), B=Multiply(A=@n42) {constB=0.04})) {name=ColorCloud, nodeColor=<rgba 0 0.9766 1 1>}
  n43 = MaterialFunctionCall!(CustomUPosition) {materialFunction=/SparksStageShow/Systems/StageManager/FXManager/MaterialFunctions/MF_MojoStageColors}
  n44 = ScalarParameter() {defaultValue=1, parameterName=UseStageColor}
  NamedRerouteDeclaration(None=LinearInterpolate(A=VectorParameter() {parameterName=Tertiary Color, defaultValue=<rgba 1 0.5953 0 1>}.RGB, B=@n43.Tertiary, Alpha=@n44)) {name=TertiaryColor, nodeColor=<rgba 0.3125 0 1 1>}
  NamedRerouteDeclaration(None=Multiply(A=@n10.G)) {name=BlurTex, nodeColor=<rgba 1 0.7031 0 1>}
  NamedRerouteDeclaration(None=LinearInterpolate(A=VectorParameter() {parameterName=PrimaryColor, defaultValue=<rgba 0.8125 0.0073 0.1938 1>}.RGB, B=@n43.Primary, Alpha=@n44)) {name=PrimaryColor, nodeColor=<rgba 0 0 0 1>}
  NamedRerouteDeclaration(None=LinearInterpolate(A=VectorParameter() {parameterName=SecondaryColor, defaultValue=<rgba 0.1116 0 1 1>}.RGB, B=@n43.Secondary, Alpha=@n44)) {name=SecondaryColor, nodeColor=<rgba 0 0 0 1>}
  NamedRerouteDeclaration(None=@n10.R) {name=PatternMask, nodeColor=<rgba 0 1 0.2734 1>}
  n45 = Multiply(A=Time() {desc=Fake Time for Testing})
  NamedRerouteDeclaration(None=LinearInterpolate(
    A=CollectionParameter() {collection=&MPC_Sparks_Audio_Parameters, parameterName=BeatLifetime}
    B=Frac(None=@n45)
    Alpha=ScalarParameter() {parameterName=Real or Test Beat}
   ) {constAlpha=0}) {name=Beat, nodeColor=<rgba 0.0312 0 1 1>}
  NamedRerouteDeclaration(None=@n33) {name=Noise, nodeColor=<rgba 0 1 0.3906 1>}
  ScalarParameter() {defaultValue=10, parameterName=Noise Pattern Mult}
 ]
}
